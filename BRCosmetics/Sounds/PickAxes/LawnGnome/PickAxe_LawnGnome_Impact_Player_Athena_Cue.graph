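# Sound cue for the Lawn Gnome pickaxe's player impact sound in Battle Royale.
# A mixer layers one of six random impact waves after a short delay with a slightly later gnome secret-laugh branch.
# Two empty slots in that branch's random node make the laugh play about one hit in three, with raised pitch and varied volume.
# It uses shared pickaxe impact attenuation and player melee weapon concurrency.
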SoundCue /BRCosmetics/Sounds/PickAxes/LawnGnome/PickAxe_LawnGnome_Impact_Player_Athena_Cue
root=Mixer_0
nodes [
 {id=Delay_0, children [Random_2], properties {delayMax=0.2, delayMin=0.15}}
 {id=Delay_1, children [Random_0], properties {delayMax=0.1, delayMin=0.1}}
 {id=Mixer_0, children [Delay_1,Delay_0], properties {inputVolume [1,1]}}
 {id=Modulator_0, children [Random_1], properties {pitchMax=1.2, pitchMin=1.1, volumeMax=1, volumeMin=0.35}}
 {
  id=Random_0
  children [WavePlayer_0, WavePlayer_1, WavePlayer_2, WavePlayer_3, WavePlayer_4, WavePlayer_10]
  properties {weights [1,1,1,1,1,1]}
 }
 {
  id=Random_1
  children [WavePlayer_9, WavePlayer_15, WavePlayer_16, WavePlayer_17]
  properties {bShouldExcludeFromBranchCulling=true, weights [1,1,1,1]}
 }
 {id=Random_2, children [Modulator_0,null,null], properties {bShouldExcludeFromBranchCulling=true, weights [1,1,1]}}
 {id=WavePlayer_0, properties {soundWaveAssetPtr=&/Athena_Pickaxe_LawnGnome_Impact_01}}
 {id=WavePlayer_1, properties {soundWaveAssetPtr=&/Athena_Pickaxe_LawnGnome_Impact_04}}
 {id=WavePlayer_10, properties {soundWaveAssetPtr=&/Athena_Pickaxe_LawnGnome_Impact_06}}
 {id=WavePlayer_15, properties {soundWaveAssetPtr=/Game/Sounds/Fort_Monster_Troll/Efforts/Gnome_SecretLaugh_03}}
 {id=WavePlayer_16, properties {soundWaveAssetPtr=/Game/Sounds/Fort_Monster_Troll/Efforts/Gnome_SecretLaugh_02}}
 {id=WavePlayer_17, properties {soundWaveAssetPtr=/Game/Sounds/Fort_Monster_Troll/Efforts/Gnome_SecretLaugh_04}}
 {id=WavePlayer_2, properties {soundWaveAssetPtr=&/Athena_Pickaxe_LawnGnome_Impact_02}}
 {id=WavePlayer_3, properties {soundWaveAssetPtr=&/Athena_Pickaxe_LawnGnome_Impact_03}}
 {id=WavePlayer_4, properties {soundWaveAssetPtr=&/Athena_Pickaxe_LawnGnome_Impact_05}}
 {id=WavePlayer_9, properties {soundWaveAssetPtr=/Game/Sounds/Fort_Monster_Troll/Efforts/Gnome_SecretLaugh_01}}
]
properties {
 attenuationSettings=/Game/Sounds/Attenuation/Fort_Pickaxe_Impact_Att
 audioPropertiesSheet=/BRCosmetics/Sounds/Settings/PropertySheets/Pickaxe/APS_Cosmetic_Pickaxe_Impact
 bPrimeOnLoad=true
 maxDistance=4300
 soundClassObject=&Fort_Impacts
 attenuationOverrides {
  bAttenuateWithLPF=true
  distanceAlgorithm=Logarithmic
  falloffDistance=800
  lPFRadiusMax=800
  lPFRadiusMin=400
  attenuationShapeExtents {x=300}
 }
 concurrencyOverrides {bLimitToOwner=true, maxCount=2, resolutionRule=StopOldest}
 concurrencySet [&Player_Melee_Weapon_Conc]
}
